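# Animation layer Blueprint for the Smackie amphibian fauna skeleton, supplying its idle override in the FaunaIdleOverride layer group.
# Fauna Idle ignores its input pose and uses a Random Sequence Player that favors the base idle over an occasional idle break, with eased blends between them.
# Fauna IdleTurn has no connected pose logic, and the main AnimGraph and EventGraph are empty.
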AnimBlueprint /Irwin/AI/Layers/Smackie/Fauna_Smackie_Idle_LayerAnimBP
target_skeleton=/Irwin/AI/Simple/Smackie/Meshes/Smackie_Amphibian_Skeleton
graphs [
 AnimGraph {nodes [Output Pose() {layerGroup=None}]}
 EventGraph {links [], nodes []}
 Fauna Idle {nodes [
   Input Pose() {name=InPose Idle}
   n1 = Random Sequence Player!() -> (Pose:Pose Link Structure) {
    is=title AnimGraphNode_RandomPlayer
    entries [
     {
      chanceToPlay=0.8
      maxLoopCount=0
      maxPlayRate=1
      minLoopCount=0
      minPlayRate=1
      sequence=/Irwin/AI/Simple/Smackie/Animations/Smackie_Idle
      blendIn {blendOption=HermiteCubic, blendTime=0.35, customCurve=null}
     }
     {
      chanceToPlay=0.2
      maxLoopCount=0
      maxPlayRate=1
      minLoopCount=0
      minPlayRate=1
      sequence=/Irwin/AI/Simple/Smackie/Animations/Smackie_IdleBreak_01
      blendIn {blendOption=HermiteCubic, blendTime=0.35, customCurve=null}
     }
    ]
   }
   Output Pose(Result=@n1) {layerGroup=FaunaIdleOverride}
  ]}
 Fauna IdleTurn {nodes [
   Input Pose() {name=InPose IdleTurn}
   Output Pose() {layerGroup=FaunaIdleOverride}
  ]}
]
variables []
